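# Translucent, unlit, two-sided Niagara sprite material for the CatSpace pickaxe effects, drawing a procedural outlined triangle.
# UV math with square-root-of-three scaling builds an outer and a smaller inner equilateral triangle, whose difference forms a pink border.
# The magenta fill carries brighter hard-edged bands, from a thresholded sine of camera-space height scrolling over time.
# Particle color tints emissive, passed through MF_FX_EmissiveAdjustments, and particle alpha scales the separate border and fill opacities.
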Material /BRCosmetics/Weapons/FORT_Melee/Pickaxe_CatSpace/FX/M_Pickaxe_CatSapce_Triangle
graph {
 properties {
  bUsedWithNiagaraSprites=true
  blendMode=BLEND_Translucent
  pixelDepthOffsetMode=PDOM_Legacy
  shadingModel=MSM_Unlit
  twoSided=true
  usedShadingModels=Unlit
 }
 nodes [
  n1 = VectorParameter() {parameterName=Color Megenta, defaultValue=<rgba 1 0 0.7297 0.2>}
  n2 = ComponentMask(None=TransformPosition(None=WorldPosition().XYZ) {transformSourceType=TRANSFORMPOSSOURCE_World, transformType=TRANSFORMPOSSOURCE_Camera}) {g=true}
  n3 = ScalarParameter() {defaultValue=0.1, parameterName=Emissive Lines Tiling, sliderMax=1}
  n4 = Time()
  n5 = Sine(None=Add(A=Multiply(A=@n2, B=@n3) {constB=0.1}, B=Multiply(A=@n4) {constB=-2}))
  n6 = LinearInterpolate(A=@n1.RGB, B=Multiply(A=@n1.RGB) {constB=2}, Alpha=If(A=Subtract(A=@n5) {constB=0.75}, "A > B"=Constant() {r=1}, "A < B"=Constant()))
  n7 = VectorParameter() {parameterName=Color Pink, defaultValue=<rgba 0.8719 0.0625 1 1>}
  n8 = SquareRoot(None=Constant() {r=3})
  n9 = Constant() {r=2}
  n10 = MaterialFunctionCall(Texture Scale=AppendVector(A=Constant() {r=1}, B=Divide(A=@n8, B=@n9))) {materialFunction=&ScaleUVsByCenter}
  n11 = ComponentMask(None=@n10.UVs) {r=true}
  n12 = Abs(None=Subtract(A=Saturate(None=@n11)) {constB=0.5})
  n13 = ComponentMask(None=@n10.UVs) {g=true}
  n14 = Constant()
  n15 = Constant() {r=1}
  n16 = Multiply(A=If(A=Multiply(A=@n12) {constB=2}, B=@n13, "A > B"=@n14, "A < B"=@n15), B=If(A=@n13, "A > B"=@n14, "A < B"=@n15) {constB=1})
  n17 = Constant() {r=1}
  n18 = Divide(A=SquareRoot(None=Constant() {r=3}), B=Constant() {r=2})
  n19 = MaterialFunctionCall(UVs=Add(A=TextureCoordinate(), B=Constant2Vector() {g=-0.05}), Texture Scale=Multiply(A=AppendVector(A=@n17, B=@n18), B=Constant() {r=0.75})) {materialFunction=&ScaleUVsByCenter}
  n20 = ComponentMask(None=@n19.UVs) {r=true}
  n21 = Abs(None=Subtract(A=Saturate(None=@n20)) {constB=0.5})
  n22 = ComponentMask(None=@n19.UVs) {g=true}
  n23 = Constant()
  n24 = Constant() {r=1}
  n25 = Multiply(A=If(A=Multiply(A=@n21) {constB=2}, B=@n22, "A > B"=@n23, "A < B"=@n24), B=If(A=@n22, "A > B"=@n23, "A < B"=@n24) {constB=1})
  n26 = Subtract(A=@n16, B=@n25)
  n27 = ParticleColor()
  MaterialFunctionCall(Emissive=Multiply(A=LinearInterpolate(A=@n6, B=@n7.RGB, Alpha=@n26), B=@n27.RGB)) {materialFunction=&MF_FX_EmissiveAdjustments}
  Multiply(A=@n27.A, B=Multiply(A=@n16, B=LinearInterpolate(A=@n7.A, B=@n1.A, Alpha=@n25))) {constA=1}
 ]
}
