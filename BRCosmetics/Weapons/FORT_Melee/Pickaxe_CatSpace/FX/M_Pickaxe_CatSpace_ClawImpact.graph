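# Translucent unlit Niagara sprite material for the claw impact effect of the CatSpace pickaxe.
# The alpha shape is the overlap of two sphere masks, with dynamic parameters Width and Pan setting the lens width and sliding it along V, and Hardness controlling the edge.
# A panning blurred noise texture, scaled by Distortion, subtly warps the mask UVs.
# Particle color drives emissive through MF_FX_EmissiveAdjustments, and particle alpha multiplies the mask.
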Material /BRCosmetics/Weapons/FORT_Melee/Pickaxe_CatSpace/FX/M_Pickaxe_CatSpace_ClawImpact
graph {
 properties {bUsedWithNiagaraSprites=true, blendMode=BLEND_Translucent, pixelDepthOffsetMode=PDOM_Legacy, shadingModel=MSM_Unlit, usedShadingModels=Unlit}
 nodes [
  n1 = ParticleColor()
  MaterialFunctionCall(Emissive=@n1.RGB) {materialFunction=&MF_FX_EmissiveAdjustments}
  n2 = NamedRerouteUsage()
  n3 = DynamicParameter() {defaultValue=<rgba 0.1 0 1 1>, paramNames [Width,Pan,Param3,Param4]}
  n4 = Subtract(B=@n3.Width) {constA=0.5}
  n5 = Subtract(A=@n2, B=AppendVector(A=@n4, B=@n3.Pan))
  n6 = Constant() {r=0.5}
  n7 = ScalarParameter() {defaultValue=0.5, parameterName=Hardness}
  n8 = Add(A=@n2, B=AppendVector(A=@n4, B=Multiply(A=@n3.Pan) {constB=-1}))
  n9 = Constant() {r=0.5}
  Multiply(A=@n1.A, B=Min(A=SphereMask(A=@n5, B=@n6, Radius=@n6, Hardness=@n7), B=SphereMask(A=@n8, B=@n9, Radius=@n9, Hardness=@n7)))
  n10 = Panner(Coordinate=TextureCoordinate()) {speedY=1}
  n11 = Constant2Vector() {g=0.5, r=3}
  n12 = ConstantBiasScale(None=TextureSample(UVs=Multiply(A=@n10, B=@n11) {constB=2}) {samplerType=SAMPLERTYPE_Grayscale, texture=/FortUILibrary/UIKit/Buttons/Textures/LowResBlurredNoise}.R) {bias=-0.5, scale=2}
  n13 = Multiply(A=ScalarParameter() {defaultValue=1, parameterName=Distortion, sliderMax=5}) {constB=0.01}
  NamedRerouteDeclaration(None=Add(A=TextureCoordinate(), B=Multiply(A=@n12, B=@n13) {constB=0.02})) {name=Distorted UV, nodeColor=<rgba 0.5234 1 0 1>}
 ]
}
