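# Translucent unlit Niagara sprite material that draws a bordered square for the CatSpace pickaxe effects.
# Two single-axis SphereMasks on U and V combine through Min to separate a magenta interior from a pink border, blending both color and alpha.
# A thresholded, time-scrolled sine of camera-space Y adds brighter lines to the interior, independent of sprite UVs.
# Particle color tints the result and scales alpha, with emissive routed through MF_FX_EmissiveAdjustments.
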Material /BRCosmetics/Weapons/FORT_Melee/Pickaxe_CatSpace/FX/M_Pickaxe_CatSpace_Square
graph {
 properties {bUsedWithNiagaraSprites=true, blendMode=BLEND_Translucent, pixelDepthOffsetMode=PDOM_Legacy, shadingModel=MSM_Unlit, usedShadingModels=Unlit}
 nodes [
  n1 = VectorParameter() {parameterName=Color Megenta, defaultValue=<rgba 1 0 0.7297 0.5>}
  n2 = ComponentMask(None=TransformPosition(None=WorldPosition().XYZ) {transformSourceType=TRANSFORMPOSSOURCE_World, transformType=TRANSFORMPOSSOURCE_Camera}) {g=true}
  n3 = ScalarParameter() {defaultValue=0.1, parameterName=Emissive Lines Tiling, sliderMax=1}
  n4 = Time()
  n5 = Sine(None=Add(A=Multiply(A=@n2, B=@n3) {constB=0.1}, B=Multiply(A=@n4) {constB=-2}))
  n6 = LinearInterpolate(A=@n1.RGB, B=Multiply(A=@n1.RGB) {constB=2}, Alpha=If(A=Subtract(A=@n5) {constB=0.75}, "A > B"=Constant() {r=1}, "A < B"=Constant()))
  n7 = VectorParameter() {parameterName=Color Pink, defaultValue=<rgba 0.8719 0.0625 1 1>}
  n8 = ComponentMask(None=TextureCoordinate()) {r=true}
  n9 = Constant() {r=0.5}
  n10 = ComponentMask(None=TextureCoordinate()) {g=true}
  n11 = Constant() {r=0.5}
  n12 = OneMinus(None=Min(A=SphereMask(A=@n8, B=@n9) {attenuationRadius=0.4}, B=SphereMask(A=@n10, B=@n11) {attenuationRadius=0.4}))
  n13 = ParticleColor()
  MaterialFunctionCall(Emissive=Multiply(A=LinearInterpolate(A=@n6, B=@n7.RGB, Alpha=@n12), B=@n13.RGB)) {materialFunction=&MF_FX_EmissiveAdjustments}
  Multiply(A=@n13.A, B=LinearInterpolate(A=@n1.A, B=@n7.A, Alpha=@n12))
 ]
}
