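# Niagara system that draws the swing trail for the CatSpace pickaxe as a single ribbon emitter following the weapon's animation.
# AnimTrail modules spawn, initialize and update short-lived particles from bone or socket positions on a skeletal mesh, with a random spawn group and ribbon ID offset.
# Emitter-age curves shape the alpha fade and drive a dynamic material parameter.
# A time-of-day scalar dims the color at night, morning and evening.
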NiagaraSystem /BRCosmetics/Weapons/FORT_Melee/Pickaxe_CatSpace/FX/NS_Pickaxe_CatSpace_Trail
dependencies {usedDataInterfaces [&NiagaraDataInterfaceCurve], usedDynamicInputs [], usedModules [&SystemState]}
emitters {Trail {
  emitterSpawnScript [
   &FN_RibbonIDOffsetUtility()
   /SetVariables(Emitter.TrailSpawnGroup:NiagaraInt32=&RandomRangeInt(Minimum=0, Maximum=10000)) {isSetParameters=true}
  ]
  emitterUpdateScript [
   &EmitterState(Life Cycle Mode=::NewEnumerator1, Loop Behavior=::NewEnumerator1, Loop Duration=1, UseLoopDelay=true)
   &FN_SpawnAnimTrail_v2()
  ]
  particleSpawnScript [
   /Game/Effects/Niagara/Modules/AnimTrail/FN_InitializeAnimTrail/FN_InitializeAnimTrailBoneIndecies(Second Bone or Socket Index=1)
   /Game/Effects/Niagara/Modules/AnimTrail/FN_InitializeAnimTrail/FN_InitializeAnimTrail(AnimTrailWidth=0.8, Color=<rgba 3 3 3 1>, Lifetime=0.3)
   &SystemLocation(MaskSpawnGroup=false)
   /SetVariables(Particles.SpawnGroup=$Emitter.TrailSpawnGroup) {isSetParameters=true}
  ]
  particleUpdateScript [
   &FN_UpdateAnimTrail_v2()
   &ScaleColor(
    Scale Alpha=&Multiply_Float(
     A=&FloatFromCurve(
      FloatCurve=DataInterface() {propertyValues {Curve {keys [
          {arriveTangent=12.9636, interpMode=RCIM_Cubic, leaveTangent=12.9636, tangentMode=RCTM_User, time=0, value=0.6}
          <at 0.1 1>
          {arriveTangent=-0.9148, interpMode=RCIM_Cubic, leaveTangent=-0.9148, tangentMode=RCTM_User, time=0.2, value=1}
          {arriveTangent=-11.3664, tangentMode=RCTM_User, time=0.4, value=0}
         ]}}}
      CurveIndex=$Emitter.Age
     )
     B=1
    )
    Scale RGB=&VectorFromFloat(Value=&FNTimeOfDayScalar(Night Scale=0.7, Morning Scale=0.85, Evening Scale=0.85))
    ScaleA=true
    ScaleRGB=true
   ) {module=ScaleColor002}
   &DynamicMaterialParameters(
    Index 0 Param 1=&FloatFromCurve(
     FloatCurve=DataInterface() {propertyValues {Curve {keys [
         {arriveTangent=0.0745, interpMode=RCIM_Cubic, leaveTangent=7.9847, tangentMode=RCTM_Break, time=0.2, value=0}
         {arriveTangent=0.0616, interpMode=RCIM_Cubic, leaveTangent=0.0616, tangentMode=RCTM_User, time=0.4, value=1}
        ]}}}
     CurveIndex=$Emitter.Age
    )
    Index 0 Param 2=&FloatFromCurve(
     FloatCurve=DataInterface() {propertyValues {Curve {keys [{arriveTangent=3.2789, time=0.05, value=-0.5}, {arriveTangent=7.5315, leaveTangent=7.5315, time=0.35, value=1.2}]}}}
     CurveIndex=$Emitter.Age
    )
    Module.Index0Param0Write=false
    Module.Index0Param1Write=true
    Module.Index0Param2Write=false
    Module.Index0Param3Write=false
   )
  ]
  rendererClasses [&NiagaraRibbonRendererProperties]
  renderers [{rendererClass=&NiagaraRibbonRendererProperties}]
 }}
summary {
 systemName=NS_Pickaxe_CatSpace_Trail
 emitters [{emitterName=Trail, rendererClasses [&NiagaraRibbonRendererProperties]}]
 userVariables [
  User.RibbonIDOffset {type=NiagaraInt32}
  User.SkeletalMesh {
   type=NiagaraDataInterfaceSkeletalMesh
   defaultValue {
    struct=NiagaraExt_VariableValue_DataInterface
    value {dataInterface=&self:NiagaraDataInterfaceSkeletalMesh_0}
   }
  }
 ]
}
